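# Additive unlit master material for Niagara sprite muzzle-flash sparks, drawn as capsules through Sprite_Capsule world position offset.
# A sphere mask shaped by dynamic-parameter Power drives a fire curve-atlas lookup, which is tinted by particle color, hue-shifted and boosted to HDR.
# The mask and particle alpha form opacity, with depth fade outside forward and mobile shading.
# A PreviousFrameSwitch subtracts per-frame particle movement from the offset, supplying motion for temporal effects.
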Material /Game/Effects/Niagara/Systems/Weapon_Masters/M_Capsule_Spark_MuzzleFlash
graph {
 properties {
  bEnableResponsiveAA=true
  bUsedWithNiagaraSprites=true
  blendMode=BLEND_Additive
  pixelDepthOffsetMode=PDOM_Legacy
  shadingModel=MSM_Unlit
  usedShadingModels=Unlit
 }
 nodes [
  n1 = StaticBool()
  n2 = MaterialFunctionCall!(
   Particle Position
   Normalized Particle Direction
   Capsule Diameter
   Capsule Pivot Alignment
   Scale Along Velocity Vector
   ---------Scale Over Velocity Parameters---------
   Particle Speed
   Minimum Speed
   Max Speed
   Minimum Scale Factor=MaterialFunctionCall(Float2=ParticleSize()) {materialFunction=&BreakOutFloat2Components}.G
   Maximum Scale Factor
   ---------- Advanced -----------
   Cap Size Multiplier=ScalarParameter() {defaultValue=0.75, parameterName=Cap Size Multiplier, sliderMax=1}
   Use Camera Offset=@n1
   Use Advanced Features=@n1
   World Position to Negate
  ) {materialFunction=/Engine/Functions/Engine_MaterialFunctions02/WorldPositionOffset/Sprite_Capsule}
  n3 = Constant() {r=0.5}
  n4 = DynamicParameter() {defaultValue=<rgba 2 20 1 1>, paramNames [Power,DepthFade,HDRMult,Param4]}
  n5 = Power(
   Base=SphereMask(A=@n2.Texture Coordinates, B=@n3, Radius=@n3, Hardness=ScalarParameter() {parameterName=SphereMaskHardness}) {hardnessPercent=0}
   Exp=@n4.Power
  )
  n6 = CurveAtlasRowParameter(CurveTime=LinearInterpolate(A=ScalarParameter() {parameterName=Glow Dark Value}, B=ScalarParameter() {defaultValue=1, parameterName=Glow Bright Value}, Alpha=@n5)) {
   atlas=/Game/Curie/Effects/temp/Curves/CA_Curie_Fire_Visuals
   curve=/Game/Curie/Effects/temp/Curves/Curve_Fiery_B
   defaultValue=8
   parameterName=Curves
  }
  n7 = ParticleColor()
  n8 = MaterialFunctionCall(Hue Shift Percentage=ScalarParameter() {parameterName=HueShift}, Texture=Multiply(A=@n6, B=@n7.RGB)) {materialFunction=&HueShift}
  MaterialFunctionCall(Emissive=MaterialFunctionCall(LDR Input=Clamp(None=@n8.Result), HDR Tint=Multiply(A=@n8.Result, B=@n7.RGB), HDR Intensity Multiplier=@n4.HDRMult) {materialFunction=&DeriveHDRfromLDR}.RemappedRange) {materialFunction=&MF_FX_EmissiveAdjustments}
  Multiply(A=@n7.A, B=ShadingPathSwitch(Default=DepthFade(Opacity=@n5, FadeDistance=@n4.DepthFade), Forward=@n5, Mobile=@n5))
  NamedRerouteDeclaration(None=@n2.World Position Offset) {name=WPO, nodeColor=<rgba 0.6172 1 0 1>}
  n9 = NamedRerouteUsage()
  n10 = Multiply(A=Multiply(A=ParticleDirection(), B=ParticleSpeed()), B=DeltaTime())
  PreviousFrameSwitch(Current Frame=@n9, Previous Frame=Subtract(A=@n9, B=Multiply(A=@n10)))
 ]
}
