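# Translucent unlit Niagara sprite material that fakes a muzzle flash decal on scene geometry behind the particle.
# World position behind the sprite, relative to particle center and size, becomes tangent-space UVs for a noise mask eroded by a dynamic parameter.
# The mask indexes a fire curve atlas, tinted, hue-shifted and remapped to HDR emissive.
# Opacity combines mask, particle alpha, squared distance falloff and sphere mask, with depth fade on higher quality non-mobile paths.
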Material /Game/Effects/Niagara/Systems/Weapon_Masters/M_ParticleDecal_MuzzleFlash
graph {
 properties {bUsedWithNiagaraSprites=true, blendMode=BLEND_Translucent, pixelDepthOffsetMode=PDOM_Legacy, shadingModel=MSM_Unlit, usedShadingModels=Unlit}
 nodes [
  n1 = Subtract(
   A=MaterialFunctionCall() {materialFunction=&WorldPositionBehindTranslucency}.Result
   B=ParticlePositionWS()
  )
  n2 = ComponentMask(None=ParticleSize()) {r=true}
  n3 = ComponentMask(None=Transform(None=Divide(A=@n1, B=@n2)) {transformSourceType=TRANSFORMSOURCE_World, transformType=TRANSFORM_Tangent}) {g=true, r=true}
  n4 = DynamicParameter() {defaultValue=<rgba 0 1 1 1>, paramNames [Erosion, World U Scale, World V Scale, Param4]}
  n5 = AppendVector(A=@n4.World U Scale, B=@n4.World U Scale)
  n6 = TextureSampleParameter2D(UVs=Subtract(A=Multiply(A=@n3, B=@n5), B=Constant2Vector() {g=0.5, r=0.5})) {parameterName=Mask Tex, texture=/Game/Effects/Fort_Effects/Textures/Fire/T_FireNoise_02}
  n7 = Subtract(A=@n6.RGBA, B=DynamicParameter() {defaultValue=<rgba 0 1 1 1>, paramNames [Erosion, World U Scale, World V Scale, Param4]}.Erosion) {constB=0.5}
  LinearInterpolate(A=@n6.RGBA, B=@n7, Alpha=ScalarParameter() {group=Panner, parameterName=UseErosion, sortPriority=3})
  n8 = StaticComponentMaskParameter(None=@n7) {defaultG=true, parameterName=Mask channel}
  n9 = CurveAtlasRowParameter(CurveTime=@n8) {
   atlas=/Game/Curie/Effects/temp/Curves/CA_Curie_Fire_Visuals
   curve=/Game/Curie/Effects/temp/Curves/Curve_Fiery_B
   defaultValue=8
   parameterName=Curves
  }
  n10 = ParticleColor()
  n11 = MaterialFunctionCall(Hue Shift Percentage=ScalarParameter() {parameterName=HueShift}, Texture=Multiply(A=@n9, B=@n10.RGB)) {materialFunction=&HueShift}
  MaterialFunctionCall(Emissive=MaterialFunctionCall(
    LDR Input=Clamp(None=@n11.Result)
    HDR Tint=Multiply(A=@n11.Result, B=@n10.RGB)
    HDR Intensity Multiplier=ScalarParameter() {defaultValue=1, group=FlameShape, parameterName=HDR_Intensity}
   ) {materialFunction=&DeriveHDRfromLDR}.RemappedRange) {materialFunction=&MF_FX_EmissiveAdjustments}
  n12 = ParticlePositionWS()
  n13 = CameraPositionWS()
  n14 = ParticleSize()
  Multiply(A=Normalize(VectorInput=Subtract(A=@n12, B=@n13)), B=Multiply(A=ComponentMask(None=@n14) {r=true}) {constB=-0.5}) {desc=WPO input}
  n15 = Divide(
   A=Distance(
    A=MaterialFunctionCall() {materialFunction=&WorldPositionBehindTranslucency}.Result
    B=ParticlePositionWS()
   )
   B=ScalarParameter() {defaultValue=5, group=Distance Fade, parameterName=Fade Distance}
  )
  n16 = ScalarParameter() {defaultValue=5, group=Distance Fade, parameterName=Fade Start Distance, sortPriority=1}
  n17 = OneMinus(None=Clamp(None=Subtract(A=@n15, B=@n16)))
  n18 = TextureCoordinate()
  n19 = Constant() {r=0.5}
  n20 = ScalarParameter() {defaultValue=0.2293, parameterName=Radius}
  n21 = ScalarParameter() {defaultValue=0.0762, parameterName=Hardness}
  n22 = Multiply(A=Multiply(A=@n10.A, B=@n8), B=Multiply(A=Multiply(A=@n17, B=@n17), B=SphereMask(A=@n18, B=@n19, Radius=@n20, Hardness=@n21)))
  n23 = DepthFade(Opacity=@n22, FadeDistance=ScalarParameter() {defaultValue=15, parameterName=DepthFadeDistance})
  QualitySwitch(Default=ShadingPathSwitch(Default=@n23, Mobile=@n22), Low=@n22, High=@n23, Medium=@n22, Epic=@n23)
 ]
}
